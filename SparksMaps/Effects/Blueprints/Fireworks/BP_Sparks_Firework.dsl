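; Firework actor for music-timed stage shows that requests a firework burst at its location on each impulse.
; Both the impulse event and function skip mobile and rate-limit triggers to the Niagara system's loop time.
; They write position, height-scaled random lifetime and a blast-only flag to a shared Niagara Data Channel for CPU Niagara, rather than spawning components.
; The remaining stage show functions are stubs returning false.
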(event EventStageShowImpulse
  (params (MusicTimeMeasure :type Float%20%28single-precision%29))
  (if (TouchPlatform|IsMobileGame)
    _
    (elif (> (- (Utilities|Time|GetGameTimeinSeconds) (Variables|Default|GetTimeLastTriggered)) (Variables|Default|GetNiagaraSystemLoopTime))
      (NiagaraDataChannel|WriteToNiagaraDataChannel(Legacy) :Channel "/SparksEffects/Effects/Fireworks/Particles/NDC_Firework.NDC_Firework" :bVisibleToBlueprint false :bVisibleToNiagaraCPU true :bVisibleToNiagaraGPU false :Position (Transformation|GetWorldLocation (Variables|Default|GetDefaultSceneRoot)) :ShouldSpawn true :LifetimeRand (* (Math|Random|RandomFloatinRange 0.75 1.15) (Variables|Default|GetHeightOverride)) :BlastOnly (Variables|Default|GetBlastOnly))
      (Variables|Default|SetTimeLastTriggered (Utilities|Time|GetGameTimeinSeconds)))))

(fn MojoStageShowActivate
  (params (Strength :type Float%20%28double-precision%29) (MusicTimeMeasure :type Float%20%28double-precision%29))
  (return false))

(fn MojoStageShowActivateBeat
  (params (Active :type Float%20%28double-precision%29) (MusicTimeMeasure :type Float%20%28double-precision%29))
  (return false))

(fn MojoStageShowColorShift
  (params (Primary :type Linear%20Color%20Structure) (Secondary :type Linear%20Color%20Structure) (Tertiary :type Linear%20Color%20Structure) (AltPrimary :type Linear%20Color%20Structure) (AltSecondary :type Linear%20Color%20Structure) (AltTertiary :type Linear%20Color%20Structure) (MusicTimeMeasure :type Float%20%28double-precision%29))
  (return false))

(fn MojoStageShowCustomMode
  (params (ModeIndex :type Integer) (MusicTimeMeasure :type Float%20%28double-precision%29))
  (return false))

(fn MojoStageShowImpulse
  (params (MusicTimeMeasure :type Float%20%28double-precision%29))
  (if (TouchPlatform|IsMobileGame)
    _
    (elif (> (- (Utilities|Time|GetGameTimeinSeconds) (Variables|Default|GetTimeLastTriggered)) (Variables|Default|GetNiagaraSystemLoopTime))
      (NiagaraDataChannel|WriteToNiagaraDataChannel(Legacy) :Channel "/SparksEffects/Effects/Fireworks/Particles/NDC_Firework.NDC_Firework" :bVisibleToBlueprint false :bVisibleToNiagaraCPU true :bVisibleToNiagaraGPU false :Position (Transformation|GetWorldLocation (Variables|Default|GetDefaultSceneRoot)) :ShouldSpawn true :LifetimeRand (* (Math|Random|RandomFloatinRange 0.75 1.15) (Variables|Default|GetHeightOverride)) :BlastOnly (Variables|Default|GetBlastOnly))
      (Variables|Default|SetTimeLastTriggered (Utilities|Time|GetGameTimeinSeconds))
      (return false))))

(fn MojoStageShowOnStop
  (return false))

(fn MojoStageShowUpdateNiagaraSystem
  (params (Niagara%20System :type Niagara%20System%20Object%20Reference))
  (return false))

(fn ConstructionScript)
